WidgetBlueprint /PlayerAugmentsUI/UI/WBP_Augment_SelectionTimer
animations [
 Anim_CountdownReveal {
  binding_count=5
  length_seconds=0.75
  object=&self:Anim_CountdownReveal
  sequence {
   bindings [
    Image_AugmentIcon_Glow {
     id=#1
     possessed_class=Image
     tracks [{
       type=MovieSceneWidgetMaterialTrack
       sections [{type=MovieSceneParameterSection, channels [Anim_BackingAlpha {type=MovieSceneScriptingFloatChannel, keys [[3,0],[9,4],[33,1]]}]}]
      }]
    }
    Image_Timer_ProgressBar {
     id=#2
     possessed_class=Image
     tracks [
      {
       type=MovieScene2DTransformTrack
       sections [{
         type=MovieScene2DTransformSection
         channels [
          Scale.X {type=MovieSceneScriptingFloatChannel, keys [[0,0.5],[7,1.125],[23,1]]}
          Scale.Y {type=MovieSceneScriptingFloatChannel, keys [[0,0.5],[7,1.125],[23,1]]}
         ]
        }]
      }
      {
       type=MovieSceneFloatTrack
       display_name=Render Opacity
       sections [{type=MovieSceneFloatSection, channels [None {type=MovieSceneScriptingFloatChannel, keys [[3,0],[9,1]]}]}]
      }
      {
       type=MovieSceneWidgetMaterialTrack
       sections [{type=MovieSceneParameterSection, channels [Ring_OutsideWidth {type=MovieSceneScriptingFloatChannel, keys [[0,0.0625],[12,0.0625],[24,0]]}]}]
      }
     ]
    }
    Image_AugmentIcon {
     id=#3
     possessed_class=Image
     tracks [{
       type=MovieScene2DTransformTrack
       sections [{
         type=MovieScene2DTransformSection
         channels [
          Scale.X {type=MovieSceneScriptingFloatChannel, keys [[3,0.9],[9,1.125],[27,1]]}
          Scale.Y {type=MovieSceneScriptingFloatChannel, keys [[3,0.9],[9,1.125],[27,1]]}
         ]
        }]
      }]
    }
    Image_Backing {
     id=#4
     possessed_class=Image
     tracks [{
       type=MovieScene2DTransformTrack
       sections [{
         type=MovieScene2DTransformSection
         channels [
          Scale.X {type=MovieSceneScriptingFloatChannel, keys [[6,1],[12,1.25],[30,1.125]]}
          Scale.Y {type=MovieSceneScriptingFloatChannel, keys [[6,1],[12,1.25],[30,1.125]]}
         ]
        }]
      }]
    }
    Image_Looper {
     id=#5
     possessed_class=Image
     tracks [{
       type=MovieSceneFloatTrack
       display_name=Render Opacity
       sections [{type=MovieSceneFloatSection, channels [None {type=MovieSceneScriptingFloatChannel, keys [[15,0],[45,1]]}]}]
      }]
    }
   ]
   display_rate {numerator=60}
  }
 }
 Anim_ProgressBar_CountdownUpdated {
  binding_count=1
  length_seconds=1
  object=&self:Anim_ProgressBar_CountdownUpdated
  sequence {
   bindings [Image_Timer_ProgressBar {
     id=#6
     possessed_class=Image
     tracks [{
       type=MovieSceneWidgetMaterialTrack
       sections [{type=MovieSceneParameterSection, channels [AnimatePercentChange {type=MovieSceneScriptingFloatChannel, keys [[0,0],[60,1]]}]}]
      }]
    }]
   display_rate {numerator=60}
  }
 }
 Anim_AugmentReady {
  binding_count=7
  length_seconds=1.1
  object=&self:Anim_AugmentReady
  sequence {
   bindings [
    Switcher_OpenAugments_Action {
     id=#7
     possessed_class=CommonVisibilitySwitcher
     tracks [{
       type=MovieSceneFloatTrack
       display_name=Render Opacity
       sections [{type=MovieSceneFloatSection, channels [None {type=MovieSceneScriptingFloatChannel, keys [[24,0],[30,0],[60,1]]}]}]
      }]
    }
    Image_Timer_InnerRing {
     id=#8
     possessed_class=Image
     tracks [
      {
       type=MovieSceneWidgetMaterialTrack
       sections [{type=MovieSceneParameterSection, channels [Animate_BackingColor {type=MovieSceneScriptingFloatChannel, const [0,2]}]}]
      }
      {
       type=MovieScene2DTransformTrack
       sections [{
         type=MovieScene2DTransformSection
         channels [
          Scale.X {type=MovieSceneScriptingFloatChannel, keys [[27,1],[36,1.333],[48,1.25]]}
          Scale.Y {type=MovieSceneScriptingFloatChannel, keys [[27,1],[36,1.333],[48,1.25]]}
         ]
        }]
      }
     ]
    }
    Image_PulseRing {
     id=#9
     possessed_class=Image
     tracks [
      {
       type=MovieSceneWidgetMaterialTrack
       sections [{
         type=MovieSceneParameterSection
         channels [
          Anim_BorderAlpha {type=MovieSceneScriptingFloatChannel, keys [[20,0],[22,1],[30,1],[66,0]]}
          Anim_Color {type=MovieSceneScriptingFloatChannel, keys [[30,0],[55,1]]}
          Border_Scale_X {type=MovieSceneScriptingFloatChannel, keys [[22,0.75],[66,1]]}
          Border_Scale_Y {type=MovieSceneScriptingFloatChannel, keys [[22,0.75],[66,1]]}
         ]
        }]
      }
      {
       type=MovieScene2DTransformTrack
       sections [{
         type=MovieScene2DTransformSection
         channels [
          Scale.X {type=MovieSceneScriptingFloatChannel, keys [[22,1],[51,3]]}
          Scale.Y {type=MovieSceneScriptingFloatChannel, keys [[22,1],[51,3]]}
         ]
        }]
      }
     ]
    }
    Overlay_CannotOpenSelector {
     id=#10
     possessed_class=Overlay
     tracks [
      {
       type=MovieSceneFloatTrack
       display_name=Render Opacity
       sections [{type=MovieSceneFloatSection, channels [None {type=MovieSceneScriptingFloatChannel, keys [[18,0],[20,1]]}]}]
      }
      {
       type=MovieScene2DTransformTrack
       sections [{
         type=MovieScene2DTransformSection
         channels [
          Scale.X {type=MovieSceneScriptingFloatChannel, keys [[18,0.5],[20,1.125],[33,1]]}
          Scale.Y {type=MovieSceneScriptingFloatChannel, keys [[18,0.5],[20,1.125],[33,1]]}
         ]
        }]
      }
     ]
    }
    Overlay_ReadyPrompt {
     id=#11
     possessed_class=Overlay
     tracks [
      {
       type=MovieSceneFloatTrack
       display_name=Render Opacity
       sections [{type=MovieSceneFloatSection, channels [None {type=MovieSceneScriptingFloatChannel, keys [[18,0],[20,1]]}]}]
      }
      {
       type=MovieScene2DTransformTrack
       sections [{
         type=MovieScene2DTransformSection
         channels [
          Scale.X {type=MovieSceneScriptingFloatChannel, keys [[18,0.5],[20,1.125],[33,1]]}
          Scale.Y {type=MovieSceneScriptingFloatChannel, keys [[18,0.5],[20,1.125],[33,1]]}
         ]
        }]
      }
     ]
    }
    Image_Timer_ProgressBar {
     id=#12
     possessed_class=Image
     tracks [
      {
       type=MovieSceneWidgetMaterialTrack
       sections [{type=MovieSceneParameterSection, channels [Ring_InsideWidth {type=MovieSceneScriptingFloatChannel, keys [[24,0.125],[30,0.333],[42,0.125]]}]}]
      }
      {
       type=MovieScene2DTransformTrack
       sections [{
         type=MovieScene2DTransformSection
         channels [
          Scale.X {type=MovieSceneScriptingFloatChannel, keys [[27.291,1],[30,1.25],[42,1]]}
          Scale.Y {type=MovieSceneScriptingFloatChannel, keys [[27.291,1],[30,1.25],[42,1]]}
         ]
        }]
      }
     ]
    }
    Image_AugmentIcon {
     id=#13
     possessed_class=Image
     tracks [{
       type=MovieScene2DTransformTrack
       sections [{
         type=MovieScene2DTransformSection
         channels [
          Scale.X {type=MovieSceneScriptingFloatChannel, keys [[30,1],[38,1.333],[51,1.25]]}
          Scale.Y {type=MovieSceneScriptingFloatChannel, keys [[30,1],[38,1.333],[51,1.25]]}
         ]
        }]
      }]
    }
   ]
   display_rate {numerator=60}
  }
 }
 Anim_AugmentReady_Reminder {
  binding_count=1
  length_seconds=0.8167
  object=&self:Anim_AugmentReady_Reminder
  sequence {bindings [Image_PulseRing {
     id=#14
     possessed_class=Image
     tracks [
      {
       type=MovieSceneWidgetMaterialTrack
       sections [{
         type=MovieSceneParameterSection
         channels [
          Anim_BorderAlpha {type=MovieSceneScriptingFloatChannel, keys [[0,0],[0.6667,1],[3.3333,1],[16.3333,0]]}
          Anim_Color {type=MovieSceneScriptingFloatChannel, keys [[3.3333,0],[11.6667,1]]}
          Border_Scale_X {type=MovieSceneScriptingFloatChannel, keys [[0.6667,0.75],[16.3333,1]]}
          Border_Scale_Y {type=MovieSceneScriptingFloatChannel, keys [[0.6667,0.75],[16.3333,1]]}
         ]
        }]
      }
      {
       type=MovieScene2DTransformTrack
       sections [{
         type=MovieScene2DTransformSection
         channels [
          Scale.X {type=MovieSceneScriptingFloatChannel, keys [[0.6667,1],[13.3333,3]]}
          Scale.Y {type=MovieSceneScriptingFloatChannel, keys [[0.6667,1],[13.3333,3]]}
         ]
        }]
      }
     ]
    }]}
 }
]
graphs [
 EventGraph {nodes [
   n1 = OnPlayerChangedBuildMode!() -> (OutputDelegate, then, bEnteredBuildMode:bool=~) {is=title K2Node_CustomEvent}
   > SetActiveWidget(
    self=Get Switcher_OpenAugments_Action() -> (Switcher_OpenAugments_Action:Common Visibility Switcher Object Reference)
    Widget=Select(
     Option 0:Widget Object Reference=Get Keybind_OpenAugments_Action() -> (Keybind_OpenAugments_Action:Keybind Widget Object Reference)
     Option 1:Widget Object Reference=Get Keybind_OpenAugments_Action_Build() -> (Keybind_OpenAugments_Action_Build:Keybind Widget Object Reference)
     Index=@n1.bEnteredBuildMode
    ) -> (ReturnValue:Widget Object Reference)
   )
   n2 = Event OnCanSelectAugmentChanged!() -> (OutputDelegate, then, bCanSelectAugment:bool) {is=title K2Node_Event}
   > Set Can Select Augment(Can Select Augment:bool=@n2.bCanSelectAugment) -> (Output_Get:bool)
   > n3 = Cast To FortPlayerController(Object=GetOwningPlayer())
   > n4 = Branch(Condition=AreTouchControlsActive!(FortPC:Fort Player Controller Object Reference=@n3) -> (ReturnValue:bool))
   SetAugmentReadyState!(execute=[@n5.then, @n4.then], self, AugmentSelectionAvailable:bool=IsAugmentSelectionAvailable!() -> (IsEnabled:bool)) -> (then)
   n5 = Branch(execute=[@n4.else, @n3.CastFailed], Condition=Get Player Able to Open Selector() -> (Player Able to Open Selector:bool))
   Event Construct()
   > n6 = GetDynamicMaterial(self=Get Image_Timer_ProgressBar() -> (Image_Timer_ProgressBar:Image Object Reference))
   > Set TimerProgressDMI(TimerProgressDMI:Material Instance Dynamic Object Reference=@n6) -> (Output_Get:Material Instance Dynamic Object Reference)
   > SetMobileVisuals()
   > UpdateVisibility()
   > n7 = Cast To FortPlayerController(Object=GetOwningPlayer())
   > Bind Event to On Player Changed Build Mode(self:Fort Player Controller Object Reference=@n7, Delegate=Create Event()) {is=title K2Node_AddDelegate}
   Event OnCountdownActivated() {is=title K2Node_Event}
   > UpdateVisibility()
   > n8 = Branch(Condition=Get bIsAugmentReadyVisible() -> (bIsAugmentReadyVisible:bool))
   PlayAnimation(execute=@n8.else, InAnimation=Get Anim_CountdownReveal() -> (Anim_CountdownReveal:Widget Animation Object Reference))
   > SetScalarParameterValue(self=Get TimerProgressDMI() -> (TimerProgressDMI:Material Instance Dynamic Object Reference), ParameterName=Percent_Start, Value=0)
   Event OnAugmentGrantingTimersFinishedForMatch() {is=title K2Node_Event}
   > Set bIsOffline(bIsOffline:bool=true) -> (Output_Get:bool)
   Event OnInitialized()
   > BindToAnimationFinished(Animation=Get Anim_AugmentReady() -> (Anim_AugmentReady:Widget Animation Object Reference), Delegate=Create Event())
   > ForceAugmentReadyState!(
    execute
    self
    AugmentSelectionAvailable:bool=CanSelectAugment!(self:Fort Player Augment HUDTimer Widget Object Reference) -> (ReturnValue:bool)
   ) -> (then)
   > Set NativeAnimProgressBarCountdownUpdated(NativeAnimProgressBarCountdownUpdated:Widget Animation Object Reference=Get Anim_ProgressBar_CountdownUpdated() -> (Anim_ProgressBar_CountdownUpdated:Widget Animation Object Reference)) -> (Output_Get:Widget Animation Object Reference)
   > Branch(Condition=IsMobileGame())
   > n9 = Async Load Asset(Asset=Get MobileCloseButtonSprite() -> (MobileCloseButtonSprite:Paper Sprite Soft Object Reference))
   n10 = Cast To PaperSprite!(execute=@n9.Completed, Object=@n9) -> (then, CastFailed, AsPaper Sprite:Paper Sprite Object Reference) {is=title K2Node_DynamicCast}
   > SetBrushFromAtlasInterface!(
    execute
    self:Image Object Reference=Get Image_MobileCloseButtonIcon() -> (Image_MobileCloseButtonIcon:Image Object Reference)
    AtlasRegion:Slate Texture Atlas Interface Interface=@n10
    bMatchSize:bool=true
   ) -> (then)
   n11 = Event OnHUDElementVisibilityChanged!() -> (OutputDelegate, then, HiddenElementTagContainer:"Gameplay Tag Container Structure (by ref)") {is=title K2Node_Event}
   > n12 = Sequence()
   n13 = HasTag!(
    TagContainer:"Gameplay Tag Container Structure (by ref)"=@n11.HiddenElementTagContainer
    Tag:Gameplay Tag Structure=(TagName="HUD.Athena.HideforVehicle")
    bExactMatch:bool=false
   ) -> (ReturnValue:bool)
   > n14 = Branch(Condition=@n13)
   > SetSize!(
    execute
    self:Spacer Object Reference=Get Spacer_WidgetOffset() -> (Spacer_WidgetOffset:Spacer Object Reference)
    InSize:Vector 2D Structure=(X=1,Y=260)
   ) -> (then)
   SetSize(execute=@n14.else, self=Get Spacer_WidgetOffset(), InSize=(X=1,Y=1))
   n15 = Cast To FortPlayerPawn(execute=@n12.then_1, Object=GetOwningPlayerPawn!(self:User Widget Object Reference) -> (ReturnValue:Pawn Object Reference))
   n16 = IsDBNO!(self:Fort Pawn Object Reference=@n15) -> (ReturnValue:bool)
   n17 = IsSkydiving!(self:Fort Player Pawn Object Reference=@n15) -> (ReturnValue:bool)
   > UpdateCanOpenSelector!(execute, self, InPlayerAbleToOpenSelector:bool=NOT Boolean(A=OR Boolean!(A=@n16, B=@n17, C:bool=@n13) -> (ReturnValue))) -> (then)
   UpdateCanOpenSelector(execute=@n15.CastFailed, InPlayerAbleToOpenSelector=NOT Boolean(A=@n13))
  ]}
 ForceAugmentReadyState {nodes [
   n1 = ForceAugmentReadyState!() -> (then, AugmentSelectionAvailable:bool=~) {is=title K2Node_FunctionEntry}
   > n2 = Set bIsAugmentReadyVisible(bIsAugmentReadyVisible:bool=@n1) -> (Output_Get:bool)
   > UpdateVisibility()
   > n3 = Branch(Condition=@n2)
   > SetVisibility(
    self=[
     Get Keybind_OpenAugments_Action_Build()
     Get Overlay_ReadyPrompt() -> (Overlay_ReadyPrompt:Overlay Object Reference)
     Get Keybind_OpenAugments_Action()
    ]
    InVisibility=SelfHitTestInvisible
   )
   > PlayAnimation(InAnimation=Get Anim_AugmentReady())
   > n4 = Branch(Condition=InEditorOrPIE!() -> (ReturnValue:bool))
   SpawnSound2D!(
    execute=@n4.else
    Sound:Sound Base Object Reference=/PlayerAugmentsCode/Sound/PlayerAugments_UI_Ready_Available_Cue
    VolumeMultiplier:float=1
    PitchMultiplier:float=1
    StartTime:float=0
    ConcurrencySettings:Sound Concurrency Object Reference
    bPersistAcrossLevelTransition:bool=false
    bAutoDestroy:bool=true
   ) -> (then, ReturnValue:Audio Component Object Reference)
   SetVisibility(
    execute=@n3.else
    self=[Get Overlay_ReadyPrompt(), Get Keybind_OpenAugments_Action(), Get Keybind_OpenAugments_Action_Build()]
    InVisibility=Collapsed
   )
   > ResetAugmentReadyStateAnims()
  ]}
 IsAugmentSelectionAvailable {nodes [
   IsAugmentSelectionAvailable() {is=title K2Node_FunctionEntry}
   n1 = Get Num Pending Augments to Grant() -> (Num Pending Augments to Grant:int=0)
   > Return Node!(execute, IsEnabled:bool=AND Boolean(A=Get Can Select Augment() -> (Can Select Augment:bool), B="integer >= integer"(A=@n1, B=1)))
  ]}
 OnAugmentReadyFadeInFinished {nodes [
   OnAugmentReadyFadeInFinished() {is=title K2Node_FunctionEntry}
   > Branch(Condition=Get bIsAugmentReadyVisible())
   > n1 = GetDynamicMaterial(self=Get Image_Timer_InnerRing() -> (Image_Timer_InnerRing:Image Object Reference))
   > SetScalarParameterValue(self=@n1, ParameterName=AutoAnimate_BackingColor, Value=1)
  ]}
 OnHideTimerEventFinished {nodes [
   OnHideTimerEventFinished() {is=title K2Node_FunctionEntry}
   > Clear and Invalidate Timer by Handle(Handle=Get HideTimerHandle() -> (HideTimerHandle:Timer Handle Structure))
  ]}
 OnNumPendingAugmentsToGrantUpdated {nodes [
   n1 = OnNumPendingAugmentsToGrantUpdated!() -> (then, NumPendingAugmentsToGrant:int=0) {is=title K2Node_FunctionEntry}
   > Set DidValueIncrease(DidValueIncrease:bool="integer > integer"(A=@n1, B=Get Num Pending Augments to Grant() -> (Num Pending Augments to Grant=0))) -> (Output_Get:bool)
   > n2 = Set Num Pending Augments to Grant(Num Pending Augments to Grant:int=@n1) -> (Output_Get:int=0)
   n3 = Get TextAugmentCount() -> (TextAugmentCount:Text)
   n4 = Get Num Pending Augments to Grant() -> (Num Pending Augments to Grant=0)
   > "SetText (Text)"(
    self=Get Text_AugmentMulitplier() -> (Text_AugmentMulitplier:Common Text Object Reference)
    InText=Select(Option 0:Text, Option 1:Text=Format Text!(Format=@n3, Count:int=@n2) -> (Result), Index="integer > integer"(A=@n4, B=1)) -> (ReturnValue:Text)
   )
   > SetAugmentReadyState(AugmentSelectionAvailable=AND Boolean(A=IsAugmentSelectionAvailable() {is=title K2Node_CallFunction}, B=Get Player Able to Open Selector()))
   > n5 = Sequence()
   > n6 = Branch(Condition="integer > integer"(A=Get Num Pending Augments to Grant() -> (Num Pending Augments to Grant=0), B=1))
   n7 = Get Image_MultiplierBacking() -> (Image_MultiplierBacking:Image Object Reference)
   n8 = Get Text_AugmentMulitplier()
   > SetVisibility(self=[@n7, @n8], InVisibility=SelfHitTestInvisible)
   > Branch(Condition=Get bIsAugmentReadyVisible())
   > n9 = Branch(Condition=InEditorOrPIE())
   SpawnSound2D(
    execute=@n9.else
    Sound=Get MultiplierSound() -> (MultiplierSound:Sound Base Object Reference)
    VolumeMultiplier=1
    PitchMultiplier=1
    StartTime=0
    bPersistAcrossLevelTransition=false
    bAutoDestroy=true
   )
   SetVisibility(execute=@n6.else, self=[@n7, @n8], InVisibility=Collapsed)
   Branch(execute=@n5.then_1, Condition=Get DidValueIncrease() -> (DidValueIncrease:bool))
   > StartHideTimerEvent()
  ]}
 PlayReminderAnimation {nodes [
   PlayReminderAnimation() {is=title K2Node_FunctionEntry}
   > Branch(Condition=Get bIsAugmentReadyVisible())
   > PlayAnimation(InAnimation=Get Anim_AugmentReady_Reminder() -> (Anim_AugmentReady_Reminder:Widget Animation Object Reference))
   > n1 = Branch(Condition=InEditorOrPIE())
   PlaySound2D(execute=@n1.else, Sound=/PlayerAugmentsCode/Sound/PlayerAugments_UI_Ready_Reminder_Cue)
  ]}
 ResetAugmentReadyStateAnims {nodes [
   ResetAugmentReadyStateAnims() {is=title K2Node_FunctionEntry}
   > n1 = Sequence()
   > n2 = GetDynamicMaterial(self=Get Image_Timer_InnerRing())
   > SetScalarParameterValue(self=@n2, ParameterName=AutoAnimate_BackingColor, Value=0)
   n3 = Get Anim_AugmentReady()
   PlayAnimation(execute=@n1.then_1, InAnimation=@n3, StartAtTime=GetEndTime(self=@n3), PlayMode=Reverse)
  ]}
 SetAugmentReadyState {nodes [
   n1 = SetAugmentReadyState!() -> (then, AugmentSelectionAvailable:bool=~) {is=title K2Node_FunctionEntry}
   > n2 = Branch(Condition="Not Equal (Boolean)"(A=@n1, B=Get bIsAugmentReadyVisible()) -> (ReturnValue=~))
   > n3 = ForceAugmentReadyState(AugmentSelectionAvailable=@n1) {is=title K2Node_CallFunction}
   n4 = Branch(execute=[@n3.then, @n2.else], Condition=Get bIsAugmentReadyVisible())
   > n5 = Set Timer by Event(Delegate=Create Event(), Time=Get ReminderFrequencyInSeconds() -> (ReminderFrequencyInSeconds:double), bLooping=true)
   > Set AugmentReadyReminderTimer(AugmentReadyReminderTimer:Timer Handle Structure=@n5) -> (Output_Get:Timer Handle Structure)
   n6 = AND Boolean(A=IsUsingTouch(PlayerController=GetOwningPlayer()), B=IsMobileGame())
   > Branch(Condition=@n6)
   > SetVisibility(self=Get Overlay_ReadyPrompt(), InVisibility=SelfHitTestInvisible)
   > SetVisibility(self=Get Overlay_CannotOpenSelector() -> (Overlay_CannotOpenSelector:Overlay Object Reference), InVisibility=Collapsed)
   Clear and Invalidate Timer by Handle(execute=@n4.else, Handle=Get AugmentReadyReminderTimer() -> (AugmentReadyReminderTimer:Timer Handle Structure))
   > Branch(Condition=@n6)
   > SetVisibility(self=Get Overlay_CannotOpenSelector(), InVisibility=SelfHitTestInvisible)
   > SetVisibility(self=Get Overlay_ReadyPrompt(), InVisibility=Collapsed)
  ]}
 SetMobileVisuals {nodes [
   SetMobileVisuals() {is=title K2Node_FunctionEntry}
   n1 = AND Boolean(A=IsUsingTouch(PlayerController=GetOwningPlayer()), B=IsMobileGame())
   > SetRenderScale!(
    execute
    self:Widget Object Reference=Get Overlay_DeviceAndProgressBar() -> (Overlay_DeviceAndProgressBar:Overlay Object Reference)
    Scale:Vector 2D Structure=Select(Option 0:Vector 2D Structure=(X=1,Y=1), Option 1:Vector 2D Structure=(X=1.45,Y=1.45), Index=@n1) -> (ReturnValue:Vector 2D Structure)
   ) -> (then)
   > SetPadding!(
    execute
    self:Horizontal Box Slot Object Reference=SlotAsHorizontalBoxSlot!(Widget:Widget Object Reference=Get Overlay_ReadyPrompt()) -> (ReturnValue:Horizontal Box Slot Object Reference)
    InPadding_Left:float=0
    InPadding_Top:float=0
    InPadding_Right:float=Select(Option 0:float=16, Option 1:float=0, Index=@n1) -> (ReturnValue:float)
    InPadding_Bottom:float=0
   ) -> (then)
   > SetRenderTranslation(
    self=Get Overlay_ReadyPrompt()
    Translation=Select(Option 0:Vector 2D Structure=(X=0,Y=0), Option 1:Vector 2D Structure=(X=-5,Y=40), Index=@n1) -> (ReturnValue:Vector 2D Structure)
   )
   > SetRenderTranslation(self=Get Overlay_CannotOpenSelector(), Translation=Select(Option 0=(X=0,Y=0), Option 1=(X=18,Y=40), Index=@n1))
   > SetRenderTranslation(
    self=Reroute Node(InputPin:Widget Object Reference) -> (OutputPin:Widget Object Reference)
    Translation=Select(Option 0=(X=0,Y=-64), Option 1=(X=0,Y=-85), Index=@n1)
   )
   > SetVisibility(
    self=Get Spacer_OverallHeightControl() -> (Spacer_OverallHeightControl:Spacer Object Reference)
    InVisibility=Select(Option 0:ESlateVisibility Enum=SelfHitTestInvisible, Option 1:ESlateVisibility Enum=Collapsed, Index=@n1) -> (ReturnValue:ESlateVisibility Enum)
   )
   > SetVisibility(InVisibility=Select(Option 0=SelfHitTestInvisible, Option 1=Collapsed, Index=@n1))
   > SetVisibility(InVisibility=Select(Option 0=SelfHitTestInvisible, Option 1=Collapsed, Index=@n1))
   > SetVisibility(InVisibility=Select(Option 0=SelfHitTestInvisible, Option 1=Collapsed, Index=@n1))
   > SetVisibility(self=Get Spacer_WidgetOffset(), InVisibility=Select(Option 0=SelfHitTestInvisible, Option 1=Collapsed, Index=@n1))
   > SetVerticalAlignment(
    self=SlotAsOverlaySlot()
    InVerticalAlignment=Select(Option 0:EVerticalAlignment Enum=VAlign_Center, Option 1:EVerticalAlignment Enum=VAlign_Fill, Index=@n1) -> (ReturnValue:EVerticalAlignment Enum)
   )
   > SetPadding(
    self:Overlay Slot Object Reference=SlotAsOverlaySlot(Widget=Get Text_Alert() -> (Text_Alert:Common Text Object Reference))
    InPadding_Left=9
    InPadding_Top=Select(Option 0:float=0, Option 1:float=2, Index=@n1) -> (ReturnValue:float)
    InPadding_Right=9
    InPadding_Bottom=0
   )
   > n2 = Branch(Condition=@n1)
   > n3 = AddChildToOverlay!(
    execute
    self:Overlay Object Reference=Get Overlay_Main() -> (Overlay_Main:Overlay Object Reference)
    Content:Widget Object Reference=Get Overlay_ReadyPrompt()
   ) -> (then, ReturnValue:Overlay Slot Object Reference)
   > SetHorizontalAlignment(self=@n3, InHorizontalAlignment=HAlign_Right)
   > SetVerticalAlignment(self=@n3, InVerticalAlignment=VAlign_Bottom)
   n4 = Get Overlay_Main()
   n5 = Get Overlay_CannotOpenSelector()
   n6 = Get Overlay_ReadyPrompt()
   n7 = Get Overlay_Main()
   Branch(
    execute=@n2.else
    Condition=AND Boolean(
     A=HasChild!(self:Panel Widget Object Reference=@n7, Content:Widget Object Reference=@n6) -> (ReturnValue:bool)
     B=HasChild(self=@n4, Content=@n5)
    )
   )
   > n8 = AddChildToHorizontalBox!(
    execute
    self:Horizontal Box Object Reference=Get HorizontalBox_Content() -> (HorizontalBox_Content:Horizontal Box Object Reference)
    Content:Widget Object Reference=Get Overlay_ReadyPrompt()
   ) -> (then, ReturnValue:Horizontal Box Slot Object Reference)
   > SetHorizontalAlignment(self:Horizontal Box Slot Object Reference=@n8, InHorizontalAlignment=HAlign_Fill)
   > SetVerticalAlignment(self:Horizontal Box Slot Object Reference=@n8, InVerticalAlignment=VAlign_Center)
   n9 = Get Overlay_CannotOpenSelector()
   > AddChildToHorizontalBox(self=Get HorizontalBox_Content(), Content=@n9)
   n10 = SlotAsOverlaySlot(Widget=@n9)
   > SetHorizontalAlignment(self:Overlay Slot Object Reference=@n10, InHorizontalAlignment=HAlign_Fill)
   > SetVerticalAlignment(self:Overlay Slot Object Reference=@n10, InVerticalAlignment=VAlign_Fill)
   > SetVisibility(self=Get Overlay_ReadyPrompt(), InVisibility=SelfHitTestInvisible)
   > SetVisibility(self=Get Overlay_CannotOpenSelector(), InVisibility=SelfHitTestInvisible)
  ]}
 StartHideTimerEvent {nodes [
   StartHideTimerEvent() {is=title K2Node_FunctionEntry}
   n1 = Get CountdownStartTimeHF() -> (CountdownStartTimeHF:Scalable Float Structure)
   > n2 = Set Timer by Event(
    Delegate=Create Event()
    Time="To Float (Integer)"(InInt=GetValueAsInteger!(Input:"Scalable Float Structure (by ref)"=@n1, Level:float=0) -> (ReturnValue:int=0))
    bLooping=false
   )
   > Set HideTimerHandle(HideTimerHandle:Timer Handle Structure=@n2) -> (Output_Get:Timer Handle Structure)
   n3 = IsMobileGame()
   n4 = IsUsingTouch(PlayerController=GetOwningPlayer())
   > n5 = Branch(Condition=NOT Boolean(A=AND Boolean(A=@n3, B=@n4)))
   > UpdateCountdown!(execute, self:Fort Player Augment HUDSelection Timer Widget Object Reference, TimeRemaining:float=0) -> (then)
   SetMobileProgressTimer!(execute=@n5.else, self:Fort Player Augment HUDSelection Timer Widget Object Reference, TimePercentage:float=0) -> (then)
  ]}
 UpdateCanOpenSelector {nodes [
   n1 = UpdateCanOpenSelector!() -> (then, InPlayerAbleToOpenSelector:bool=~) {is=title K2Node_FunctionEntry}
   > n2 = Set Player Able to Open Selector(Player Able to Open Selector:bool=@n1) -> (Output_Get:bool)
   > n3 = Branch(Condition=@n2)
   > SetAugmentReadyState(AugmentSelectionAvailable=IsAugmentSelectionAvailable()) {is=title K2Node_CallFunction}
   > UpdateVisibility()
   Set bIsAugmentReadyVisible(execute=@n3.else, bIsAugmentReadyVisible=false)
   > UpdateVisibility()
   > Clear and Invalidate Timer by Handle(Handle=Get AugmentReadyReminderTimer())
  ]}
 UpdateMobileIcon {nodes [
   n1 = UpdateMobileIcon!() -> (then, IsSelectionToolOpened:bool=~) {is=title K2Node_FunctionEntry}
   n2 = Select(Option 0:ESlateVisibility Enum=SelfHitTestInvisible, Option 1:ESlateVisibility Enum=Collapsed, Index=@n1) -> (ReturnValue:ESlateVisibility Enum)
   > SetVisibility(self=Get HorizontalBox_Content(), InVisibility=@n2)
   > SetVisibility(InVisibility=@n2)
   > SetVisibility(self=Get Image_MobileCloseButtonIcon(), InVisibility=Select(Option 0=Collapsed, Option 1=SelfHitTestInvisible, Index=@n1))
  ]}
 UpdateVisibility {nodes [
   UpdateVisibility() {is=title K2Node_FunctionEntry}
   n1 = GetOwningPlayer()
   > n2 = Branch(Condition=AND Boolean(A=IsUsingTouch(PlayerController=@n1), B=IsMobileGame()))
   > SetVisibility(self=Get Overlay_Main(), InVisibility=Select(Option 0=Collapsed, Option 1=SelfHitTestInvisible, Index=Get Player Able to Open Selector()))
   n3 = IsCountdownVisible!(self:Fort Player Augment HUDSelection Timer Widget Object Reference) -> (ReturnValue:bool)
   n4 = Get bIsAugmentReadyVisible()
   SetVisibility(
    execute=@n2.else
    self=Get Overlay_Main()
    InVisibility=Select(Option 0=Collapsed, Option 1=SelfHitTestInvisible, Index=OR Boolean!(A=@n3, B=@n4) -> (ReturnValue))
   )
  ]}
]
variables [
 AugmentReadyReminderTimer {type=TimerHandle}
 Can Select Augment {type=boolean}
 MobileCloseButtonSprite {type=PaperSprite, default=/Game/Athena/HUD/Atlas/Bacchus/Cancel_Normal_Sprite}
 MultiplierSound {type=SoundBase, default=/PlayerAugmentsUI/Sound/PlayerAugments_UI_Ready_Available_Multiplier_Cue}
 Num Pending Augments to Grant {type=integer}
 OwningPerkComponent {type=FortControllerComponent_SpyTechPerkSystem, default=null}
 Player Able to Open Selector {type=boolean}
 ReminderFrequencyInSeconds {type=number, default=20}
 TextAugmentCount {type=string, default="x{Count}"}
 TextAugmentReady {type=string, default="Ready!"}
]
widgets {
 info {
  inheritedWidgetCount=1
  parentClass=/Script/PlayerAugmentsUI.FortPlayerAugmentHUDSelectionTimerWidget
  rootWidgetClass=&VerticalBox
  widgetCount=35
 }
 widgets [
  VerticalBox_1 {widgetClassPath=&VerticalBox, widget=VerticalBox_1}
  Overlay_Main {
   widgetClassPath=&Overlay
   bIsVariable=true
   parent=VerticalBox_1
   widget=Overlay_Main
   properties {visibility=HitTestInvisible}
   slot {class=VerticalBoxSlot}
  }
  Spacer_OverallHeightControl {
   widgetClassPath=&Spacer
   bIsVariable=true
   parent=Overlay_Main
   widget=Spacer_OverallHeightControl
   properties {size=<0 128>}
   slot {class=OverlaySlot, properties {horizontalAlignment=HAlign_Right, verticalAlignment=VAlign_Center}}
  }
  Border_CHVB_TouchButton {
   widgetClassPath=&CommonHardwareVisibilityBorder
   parent=Overlay_Main
   widget=Border_CHVB_TouchButton
   properties {
    horizontalAlignment=HAlign_Center
    verticalAlignment=VAlign_Center
    visibilityQuery {
     autoDescription=" ANY( Input.Touch )"
     tokenStreamVersion=0
     userDescription=Touch-Only
     queryTokenStream [0,1,1,1,0]
     tagDictionary [{tagName=Input.Touch}]
    }
   }
   slot {class=OverlaySlot, properties {horizontalAlignment=HAlign_Right, verticalAlignment=VAlign_Fill}}
  }
  Overlay_0 {
   widgetClassPath=&Overlay
   parent=Border_CHVB_TouchButton
   widget=Overlay_0
   slot {class=BorderSlot, properties {horizontalAlignment=HAlign_Center, verticalAlignment=VAlign_Center, padding=<ltrb 0 0 0 0>}}
  }
  Image_MobileButtonBG {
   widgetClassPath=&Image
   bIsVariable=true
   parent=Overlay_0
   widget=Image_MobileButtonBG
   properties {
    brush {
     bIsDynamicallyLoaded=false
     drawAs=Image
     imageType=FullColor
     mirroring=NoMirror
     resourceName=None
     resourceObject=/Game/Athena/HUD/Bacchus/Material/MI_UI_ButtonBG_Dark
     tiling=NoTile
     imageSize=<128 128>
     margin=<ltrb 0 0 0 0>
     outlineSettings {
      bUseBrushTransparency=false
      roundingType=HalfHeightRadius
      width=0
      color {colorUseRule=UseColor_Specified, specifiedColor=<rgba 0 0 0 0>}
      cornerRadii=<0 0 0 0>
     }
     tintColor {colorUseRule=UseColor_Specified, specifiedColor=<rgba 1 1 1 1>}
     uVRegion {bIsValid=false, max=<0 0>, min=<0 0>}
    }
    renderTransform {angle=0, scale=<1.1 1.1>, shear=<0 0>, translation=<-4 0>}
   }
   slot {class=OverlaySlot}
  }
  Image_MobileCloseButtonIcon {
   widgetClassPath=&Image
   bIsVariable=true
   parent=Overlay_0
   widget=Image_MobileCloseButtonIcon
   properties {
    visibility=Collapsed
    brush {
     bIsDynamicallyLoaded=false
     drawAs=Image
     imageType=NoImage
     mirroring=NoMirror
     resourceName=None
     resourceObject=null
     tiling=NoTile
     imageSize=<64 64>
     margin=<ltrb 0 0 0 0>
     outlineSettings {
      bUseBrushTransparency=false
      roundingType=HalfHeightRadius
      width=0
      color {colorUseRule=UseColor_Specified, specifiedColor=<rgba 0 0 0 0>}
      cornerRadii=<0 0 0 0>
     }
     tintColor {colorUseRule=UseColor_Specified, specifiedColor=<rgba 1 1 1 1>}
     uVRegion {bIsValid=false, max=<0 0>, min=<0 0>}
    }
    colorAndOpacity=<rgba 1 1 1 0.6>
    renderTransform {angle=0, scale=<1.1 1.1>, shear=<0 0>, translation=<-4 0>}
   }
   slot {class=OverlaySlot, properties {horizontalAlignment=HAlign_Center, verticalAlignment=VAlign_Center}}
  }
  GridPanel_92 {
   widgetClassPath=&GridPanel
   parent=Overlay_Main
   widget=GridPanel_92
   slot {class=OverlaySlot, properties {horizontalAlignment=HAlign_Right, verticalAlignment=VAlign_Center}}
  }
  Switcher_OpenAugments_Action {
   widgetClassPath=&CommonVisibilitySwitcher
   bIsVariable=true
   parent=GridPanel_92
   widget=Switcher_OpenAugments_Action
   slot {class=GridSlot, properties {horizontalAlignment=HAlign_Right, verticalAlignment=VAlign_Center, padding=<ltrb 0 0 20 0>}}
  }
  Keybind_OpenAugments_Action {
   widgetClassPath=/FortUILibrary/Input/KeybindWidget
   bIsVariable=true
   parent=Switcher_OpenAugments_Action
   widget=Keybind_OpenAugments_Action
   properties {
    boundAction=OpenAugmentSelection
    brushSize=VerySmall
    visibility=Collapsed
    boundActionFallbacks [GamepadSquadQuickChatPicker, GamepadEmotePickerOrReplayLastEmoteHold]
    renderTransform {angle=0, scale=<0.825 0.825>, shear=<0 0>, translation=<0 0>}
   }
   slot {class=CommonVisibilitySwitcherSlot}
  }
  Keybind_OpenAugments_Action_Build {
   widgetClassPath=&KeybindWidget
   bIsVariable=true
   parent=Switcher_OpenAugments_Action
   widget=Keybind_OpenAugments_Action_Build
   properties {
    boundAction=OpenAugmentSelection
    brushSize=VerySmall
    visibility=Collapsed
    boundActionFallbacks [GamepadSquadQuickChatPickerWhileBuilding, GamepadEmotePickerOrReplayLastEmoteHold]
    renderTransform {angle=0, scale=<0.825 0.825>, shear=<0 0>, translation=<0 0>}
   }
   slot {class=CommonVisibilitySwitcherSlot}
  }
  Overlay_DeviceAndProgressBar {
   widgetClassPath=&Overlay
   bIsVariable=true
   parent=GridPanel_92
   widget=Overlay_DeviceAndProgressBar
   slot {class=GridSlot, properties {column=1}}
  }
  Image_Backing {
   widgetClassPath=&Image
   parent=Overlay_DeviceAndProgressBar
   widget=Image_Backing
   properties {
    visibility=SelfHitTestInvisible
    brush {
     bIsDynamicallyLoaded=false
     drawAs=Image
     imageType=NoImage
     mirroring=NoMirror
     resourceName=None
     resourceObject=/PlayerAugmentsUI/Materials/M_UI_AugmentTimer_BackingPlate
     tiling=NoTile
     imageSize=<32 32>
     margin=<ltrb 0 0 0 0>
     outlineSettings {
      bUseBrushTransparency=false
      roundingType=HalfHeightRadius
      width=0
      color {colorUseRule=UseColor_Specified, specifiedColor=<rgba 0 0 0 0>}
      cornerRadii=<0 0 0 0>
     }
     tintColor {colorUseRule=UseColor_Specified, specifiedColor=<rgba 1 1 1 1>}
     uVRegion {bIsValid=false, max=<0 0>, min=<0 0>}
    }
    renderTransform {angle=0, scale=<1.125 1.125>, shear=<0 0>, translation=<0 0>}
   }
   slot {class=OverlaySlot, properties {horizontalAlignment=HAlign_Fill, verticalAlignment=VAlign_Fill}}
  }
  Image_Looper {
   widgetClassPath=&Image
   parent=Overlay_DeviceAndProgressBar
   widget=Image_Looper
   properties {
    visibility=SelfHitTestInvisible
    brush {
     bIsDynamicallyLoaded=false
     drawAs=Image
     imageType=NoImage
     mirroring=NoMirror
     resourceName=None
     resourceObject=/Game/UI/Foundation/Materials/MI_Throbber_Thin
     tiling=NoTile
     imageSize=<64 64>
     margin=<ltrb 0 0 0 0>
     outlineSettings {
      bUseBrushTransparency=false
      roundingType=HalfHeightRadius
      width=0
      color {colorUseRule=UseColor_Specified, specifiedColor=<rgba 0 0 0 0>}
      cornerRadii=<0 0 0 0>
     }
     tintColor {colorUseRule=UseColor_Specified, specifiedColor=<rgba 1 1 1 1>}
     uVRegion {bIsValid=false, max=<0 0>, min=<0 0>}
    }
    colorAndOpacity=<rgba 0.6667 0 1 1>
    renderTransform {angle=0, scale=<1.25 1.25>, shear=<0 0>, translation=<0 0>}
   }
   slot {class=OverlaySlot, properties {horizontalAlignment=HAlign_Fill, verticalAlignment=VAlign_Fill}}
  }
  Image_PulseRing {
   widgetClassPath=&Image
   parent=Overlay_DeviceAndProgressBar
   widget=Image_PulseRing
   properties {
    visibility=SelfHitTestInvisible
    brush {
     bIsDynamicallyLoaded=false
     drawAs=Image
     imageType=NoImage
     mirroring=NoMirror
     resourceName=None
     resourceObject=/PlayerAugmentsUI/Materials/MI_UI_PlayerAugment_DeviceReady_Countdown_PulseRing
     tiling=NoTile
     imageSize=<16 16>
     margin=<ltrb 0 0 0 0>
     outlineSettings {
      bUseBrushTransparency=false
      roundingType=HalfHeightRadius
      width=0
      color {colorUseRule=UseColor_Specified, specifiedColor=<rgba 0 0 0 0>}
      cornerRadii=<0 0 0 0>
     }
     tintColor {colorUseRule=UseColor_Specified, specifiedColor=<rgba 1 1 1 1>}
     uVRegion {bIsValid=false, max=<0 0>, min=<0 0>}
    }
    colorAndOpacity=<rgba 1 1 1 0.995>
    renderTransform {angle=0, scale=<1.25 1.25>, shear=<0 0>, translation=<0 0>}
   }
   slot {class=OverlaySlot, properties {horizontalAlignment=HAlign_Fill, verticalAlignment=VAlign_Fill}}
  }
  Image_AugmentIcon_Glow {
   widgetClassPath=&Image
   parent=Overlay_DeviceAndProgressBar
   widget=Image_AugmentIcon_Glow
   properties {
    visibility=SelfHitTestInvisible
    brush {
     bIsDynamicallyLoaded=false
     drawAs=Image
     imageType=NoImage
     mirroring=NoMirror
     resourceName=None
     resourceObject=/PlayerAugmentsUI/Materials/MI_UI_PlayerAugment_Device_RadialGlow
     tiling=NoTile
     imageSize=<32 32>
     margin=<ltrb 0 0 0 0>
     outlineSettings {
      bUseBrushTransparency=false
      roundingType=HalfHeightRadius
      width=0
      color {colorUseRule=UseColor_Specified, specifiedColor=<rgba 0 0 0 0>}
      cornerRadii=<0 0 0 0>
     }
     tintColor {colorUseRule=UseColor_Specified, specifiedColor=<rgba 1 1 1 1>}
     uVRegion {bIsValid=false, max=<0 0>, min=<0 0>}
    }
    renderTransform {angle=0, scale=<6 6>, shear=<0 0>, translation=<0 0>}
   }
   slot {class=OverlaySlot, properties {horizontalAlignment=HAlign_Center, verticalAlignment=VAlign_Center}}
  }
  Image_Timer_ProgressBar {
   widgetClassPath=&Image
   bIsVariable=true
   parent=Overlay_DeviceAndProgressBar
   widget=Image_Timer_ProgressBar
   properties {
    visibility=SelfHitTestInvisible
    brush {
     bIsDynamicallyLoaded=false
     drawAs=Image
     imageType=NoImage
     mirroring=NoMirror
     resourceName=None
     resourceObject=/PlayerAugmentsUI/Materials/M_UI_AugmentTimer
     tiling=NoTile
     imageSize=<72 72>
     margin=<ltrb 0 0 0 0>
     outlineSettings {
      bUseBrushTransparency=false
      roundingType=HalfHeightRadius
      width=0
      color {colorUseRule=UseColor_Specified, specifiedColor=<rgba 0 0 0 0>}
      cornerRadii=<0 0 0 0>
     }
     tintColor {colorUseRule=UseColor_Specified, specifiedColor=<rgba 1 1 1 1>}
     uVRegion {bIsValid=false, max=<0 0>, min=<0 0>}
    }
   }
   slot {class=OverlaySlot, properties {horizontalAlignment=HAlign_Fill, verticalAlignment=VAlign_Fill}}
  }
  Image_Timer_InnerRing {
   widgetClassPath=&Image
   bIsVariable=true
   parent=Overlay_DeviceAndProgressBar
   widget=Image_Timer_InnerRing
   properties {
    visibility=SelfHitTestInvisible
    brush {
     bIsDynamicallyLoaded=false
     drawAs=Image
     imageType=NoImage
     mirroring=NoMirror
     resourceName=None
     resourceObject=/PlayerAugmentsUI/Materials/M_UI_AugmentTimer_InnerRing
     tiling=NoTile
     imageSize=<48 48>
     margin=<ltrb 0 0 0 0>
     outlineSettings {
      bUseBrushTransparency=false
      roundingType=HalfHeightRadius
      width=0
      color {colorUseRule=UseColor_Specified, specifiedColor=<rgba 0 0 0 0>}
      cornerRadii=<0 0 0 0>
     }
     tintColor {colorUseRule=UseColor_Specified, specifiedColor=<rgba 1 1 1 1>}
     uVRegion {bIsValid=false, max=<0 0>, min=<0 0>}
    }
   }
   slot {class=OverlaySlot, properties {horizontalAlignment=HAlign_Center, verticalAlignment=VAlign_Center}}
  }
  Image_AugmentIcon {
   widgetClassPath=&Image
   parent=Overlay_DeviceAndProgressBar
   widget=Image_AugmentIcon
   properties {
    visibility=SelfHitTestInvisible
    brush {
     bIsDynamicallyLoaded=false
     drawAs=Image
     imageType=NoImage
     mirroring=NoMirror
     resourceName=None
     resourceObject=/PlayerAugmentsUI/Textures/T_UI_PlayerAugment_Icon
     tiling=NoTile
     imageSize=<40 40>
     margin=<ltrb 0 0 0 0>
     outlineSettings {
      bUseBrushTransparency=false
      roundingType=HalfHeightRadius
      width=0
      color {colorUseRule=UseColor_Specified, specifiedColor=<rgba 0 0 0 0>}
      cornerRadii=<0 0 0 0>
     }
     tintColor {colorUseRule=UseColor_Specified, specifiedColor=<rgba 1 1 1 1>}
     uVRegion {bIsValid=false, max=<0 0>, min=<0 0>}
    }
    renderTransform {angle=-8, scale=<1 1>, shear=<0 0>, translation=<0.5 -1>}
   }
   slot {class=OverlaySlot, properties {horizontalAlignment=HAlign_Center, verticalAlignment=VAlign_Center}}
  }
  Switcher_ReadyPrompt {
   widgetClassPath=&WidgetSwitcher
   bIsVariable=true
   parent=GridPanel_92
   widget=Switcher_ReadyPrompt
   slot {class=GridSlot, properties {column=1, horizontalAlignment=HAlign_Center, row=1, verticalAlignment=VAlign_Center, padding=<ltrb 0 10 0 0>}}
  }
  Overlay_ReadyOpenPrompt {
   widgetClassPath=&Overlay
   bIsVariable=true
   parent=Switcher_ReadyPrompt
   widget=Overlay_ReadyOpenPrompt
   slot {class=WidgetSwitcherSlot}
  }
  HorizontalBox_0 {widgetClassPath=&HorizontalBox, parent=Overlay_ReadyOpenPrompt, widget=HorizontalBox_0, slot {class=OverlaySlot}}
  Overlay_ReadyPrompt {widgetClassPath=&Overlay, bIsVariable=true, parent=HorizontalBox_0, widget=Overlay_ReadyPrompt, slot {class=HorizontalBoxSlot}}
  Image_MultiplierBacking {
   widgetClassPath=&Image
   bIsVariable=true
   parent=Overlay_ReadyPrompt
   widget=Image_MultiplierBacking
   properties {brush {
     bIsDynamicallyLoaded=false
     drawAs=Box
     imageType=NoImage
     mirroring=NoMirror
     resourceName=None
     resourceObject=/PlayerAugmentsUI/Materials/MI_UI_PlayerAugment_Device_MultiplierBacking
     tiling=NoTile
     imageSize=<28 28>
     margin=<ltrb 0.5 0.5 0.5 0.5>
     outlineSettings {
      bUseBrushTransparency=false
      roundingType=HalfHeightRadius
      width=0
      color {colorUseRule=UseColor_Specified, specifiedColor=<rgba 0 0 0 0>}
      cornerRadii=<0 0 0 0>
     }
     tintColor {colorUseRule=UseColor_Specified, specifiedColor=<rgba 1 1 1 1>}
     uVRegion {bIsValid=false, max=<0 0>, min=<0 0>}
    }}
   slot {class=OverlaySlot, properties {horizontalAlignment=HAlign_Fill, verticalAlignment=VAlign_Fill, padding=<ltrb 0 -2 0 -2>}}
  }
  Spacer_246 {
   widgetClassPath=&Spacer
   parent=Overlay_ReadyPrompt
   widget=Spacer_246
   properties {size=<1 28>}
   slot {class=OverlaySlot, properties {verticalAlignment=VAlign_Center}}
  }
  Text_AugmentMulitplier {
   widgetClassPath=&CommonTextBlock
   bIsVariable=true
   parent=Overlay_ReadyPrompt
   widget=Text_AugmentMulitplier
   properties {
    justification=Center
    mobileFontSizeMultiplier=1.4
    text=x2
    colorAndOpacity {colorUseRule=UseColor_Specified, specifiedColor=<rgba 0.4348 0 1 1>}
    font {
     bForceMonospaced=false
     bMaterialIsStencil=false
     fontMaterial=null
     fontObject=&Burbank
     letterSpacing=0
     monospacedWidth=1
     size=13
     skewAmount=0
     typefaceFontName=BurbankRegular-Black
     outlineSettings {
      bApplyOutlineToDropShadows=false
      bMiteredCorners=false
      bSeparateFillAlpha=true
      outlineMaterial=null
      outlineSize=0
      outlineColor=<rgba 0 0 0 0.5>
     }
    }
    shadowOffset=<0 0>
   }
   slot {class=OverlaySlot, properties {horizontalAlignment=HAlign_Fill, verticalAlignment=VAlign_Center, padding=<ltrb 8 0 6 0>}}
  }
  Overlay_CannotOpenSelector {
   widgetClassPath=&Overlay
   bIsVariable=true
   parent=Switcher_ReadyPrompt
   widget=Overlay_CannotOpenSelector
   slot {class=WidgetSwitcherSlot}
  }
  Image_AlertGlow {
   widgetClassPath=&Image
   parent=Overlay_CannotOpenSelector
   widget=Image_AlertGlow
   properties {
    visibility=SelfHitTestInvisible
    brush {
     bIsDynamicallyLoaded=false
     drawAs=Box
     imageType=NoImage
     mirroring=NoMirror
     resourceName=None
     resourceObject=/PlayerAugmentsUI/Materials/MI_UI_PlayerAugment_Alert_Glow
     tiling=NoTile
     imageSize=<32 32>
     margin=<ltrb 0.5 0.5 0.5 0.5>
     outlineSettings {
      bUseBrushTransparency=false
      roundingType=HalfHeightRadius
      width=0
      color {colorUseRule=UseColor_Specified, specifiedColor=<rgba 0 0 0 0>}
      cornerRadii=<0 0 0 0>
     }
     tintColor {colorUseRule=UseColor_Specified, specifiedColor=<rgba 1 1 1 1>}
     uVRegion {bIsValid=false, max=<0 0>, min=<0 0>}
    }
   }
   slot {class=OverlaySlot, properties {horizontalAlignment=HAlign_Fill, verticalAlignment=VAlign_Fill, padding=<ltrb -10 -10 -10 -10>}}
  }
  Image_AlertBacking {
   widgetClassPath=&Image
   parent=Overlay_CannotOpenSelector
   widget=Image_AlertBacking
   properties {
    visibility=SelfHitTestInvisible
    brush {
     bIsDynamicallyLoaded=false
     drawAs=Box
     imageType=NoImage
     mirroring=NoMirror
     resourceName=None
     resourceObject=/PlayerAugmentsUI/Materials/MI_UI_PlayerAugment_Alert_Backing
     tiling=NoTile
     imageSize=<16 16>
     margin=<ltrb 0.5 0.5 0.5 0.5>
     outlineSettings {
      bUseBrushTransparency=false
      roundingType=HalfHeightRadius
      width=0
      color {colorUseRule=UseColor_Specified, specifiedColor=<rgba 0 0 0 0>}
      cornerRadii=<0 0 0 0>
     }
     tintColor {colorUseRule=UseColor_Specified, specifiedColor=<rgba 1 1 1 1>}
     uVRegion {bIsValid=false, max=<0 0>, min=<0 0>}
    }
   }
   slot {class=OverlaySlot, properties {horizontalAlignment=HAlign_Fill, verticalAlignment=VAlign_Fill}}
  }
  Text_Alert {
   widgetClassPath=&CommonTextBlock
   bIsVariable=true
   parent=Overlay_CannotOpenSelector
   widget=Text_Alert
   properties {
    mobileFontSizeMultiplier=1.3
    text=Unavailable
    textTransformPolicy=ToUpper
    font {
     bForceMonospaced=false
     bMaterialIsStencil=false
     fontMaterial=/PlayerAugmentsUI/Materials/MI_UI_PlayerAugment_Alert_Font
     fontObject=&BurbankSmall
     letterSpacing=0
     monospacedWidth=1
     size=13
     skewAmount=0
     typefaceFontName=Black
     outlineSettings {
      bApplyOutlineToDropShadows=false
      bMiteredCorners=false
      bSeparateFillAlpha=true
      outlineMaterial=null
      outlineSize=2
      outlineColor=<rgba 0 0 0 0>
     }
    }
    shadowOffset=<0 0>
   }
   slot {class=OverlaySlot, properties {verticalAlignment=VAlign_Center, padding=<ltrb 9 0 9 0>}}
  }
  HorizontalBox_Content {
   widgetClassPath=&HorizontalBox
   bIsVariable=true
   parent=Overlay_Main
   widget=HorizontalBox_Content
   slot {class=OverlaySlot, properties {horizontalAlignment=HAlign_Right, verticalAlignment=VAlign_Center}}
  }
  VerticalBox_2 {
   widgetClassPath=&VerticalBox
   parent=HorizontalBox_Content
   widget=VerticalBox_2
   slot {class=HorizontalBoxSlot, properties {horizontalAlignment=HAlign_Center, verticalAlignment=VAlign_Center}}
  }
  Spacer_ScreenEdgePadding {
   widgetClassPath=&Spacer
   parent=HorizontalBox_Content
   widget=Spacer_ScreenEdgePadding
   properties {size=<32 0>}
   slot {class=HorizontalBoxSlot}
  }
  Spacer_WidgetOffset {widgetClassPath=&Spacer, bIsVariable=true, parent=VerticalBox_1, widget=Spacer_WidgetOffset, slot {class=VerticalBoxSlot}}
  TutorialHighlight {widgetClassPath=&TutorialHighlight}
 ]
}
