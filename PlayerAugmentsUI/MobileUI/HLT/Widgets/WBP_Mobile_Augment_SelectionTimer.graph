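# Mobile HUD widget for the player augment selection timer, built on FortMobilePlayerAugmentTimer.
# It wraps the shared WBP_Augment_SelectionTimer in a scaled, sized overlay alongside an inherited, transparent button.
# On HUD visibility changes, it passes the selection-item equipped state to the timer's mobile icon update.
# It then collapses itself when force-collapsed or when hidden-element tags match a hardcoded set (gliding, free fall, DBNO, bus, spectate, post-game), otherwise showing itself.
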WidgetBlueprint /PlayerAugmentsUI/MobileUI/HLT/Widgets/WBP_Mobile_Augment_SelectionTimer
graphs [EventGraph {nodes [
   n1 = Event On HUD Element Visibility Changed!() -> (
    OutputDelegate
    then
    HiddenElementTagContainer:"Gameplay Tag Container Structure (by ref)"
    bIsSelectionItemEquipped:bool
    bForceCollapse:bool
   ) {is=title K2Node_Event}
   > UpdateMobileIcon!(
    execute
    self:WBP Augment Selection Timer Object Reference=Get WBP_Augment_SelectionTimer_34() -> (WBP_Augment_SelectionTimer_34:WBP Augment Selection Timer Object Reference)
    IsSelectionToolOpened:bool=@n1.bIsSelectionItemEquipped
   ) -> (then)
   n2 = OR Boolean(
    A=@n1.bForceCollapse
    B=HasAnyTags!(
     TagContainer:"Gameplay Tag Container Structure (by ref)"=@n1.HiddenElementTagContainer
     OtherContainer:"Gameplay Tag Container Structure (by ref)"=MakeLiteralGameplayTagContainer!(Value:Gameplay Tag Container Structure=(GameplayTags=((TagName="HUD.Visibility.FreeFalling"),(TagName="HUD.Visibility.IsGliding"),(TagName="Gameplay.Action.Player.DBNOAthena"),(TagName="HUD.Athena.HideOnSpectate"),(TagName="HUD.Athena.HideForPostGame"),(TagName="HUD.Athena.HideForPostGameSpectate"),(TagName="HUD.Athena.HideOnBus")))) -> (ReturnValue:Gameplay Tag Container Structure)
     bExactMatch:bool=false
    ) -> (ReturnValue:bool)
   )
   > SetVisibility(InVisibility=Select(Option 0=Collapsed, Option 1=Visible, Index=NOT Boolean(A=@n2)))
  ]}]
variables []
widgets {
 info {
  inheritedWidgetCount=4
  parentClass=/Script/PlayerAugmentsUI.FortMobilePlayerAugmentTimer
  rootWidgetClass=&ScaleBox
  widgetCount=7
 }
 widgets [
  ScaleBox_HUDWidgetScale {widgetClassPath=&ScaleBox, bInherited=true, widget=ScaleBox_HUDWidgetScale, properties {stretch=UserSpecified}}
  SizeBox_3 {
   widgetClassPath=&SizeBox
   parent=ScaleBox_HUDWidgetScale
   widget=SizeBox_3
   properties {bOverride_MaxDesiredWidth=true, heightOverride=128, maxDesiredHeight=340, maxDesiredWidth=175, widthOverride=128}
   slot {class=ScaleBoxSlot}
  }
  Overlay_34 {
   widgetClassPath=&Overlay
   parent=SizeBox_3
   widget=Overlay_34
   slot {class=SizeBoxSlot, properties {horizontalAlignment=HAlign_Center, verticalAlignment=VAlign_Center}}
  }
  WBP_Augment_SelectionTimer_34 {
   widgetClassPath=/PlayerAugmentsUI/UI/WBP_Augment_SelectionTimer
   bIsVariable=true
   parent=Overlay_34
   widget=WBP_Augment_SelectionTimer_34
   properties {
    hUDElementTag {gameplayTags [
      {tagName=HUD.Athena.HideForPostGame}
      {tagName=HUD.Athena.HideforRadial}
      {tagName=HUD.Athena.HideOnBus}
      {tagName=HUD.Athena.Hide.WithMap}
      {tagName=HUD.Hide.WhileLoading}
      {tagName=Athena.Creative.HideForLoading}
      {tagName=Athena.HideForCapture}
      {tagName=Athena.Creative.HUDController.HidePlayerInventory}
      {tagName=HUD.Athena.Hide.WithInventory}
      {tagName=HUD.Athena.Hide.WithRCPawn}
      {tagName=HUD.Athena.HideForAbility}
      {tagName=HUD.Athena.HideForGadget}
      {tagName=HUD.Event.HideForSpecialEvent}
      {tagName=HUD.Athena.HideForConversation}
      {tagName=HUD.Athena.HideOnSpectate}
     ]}
    renderTransform {angle=0, scale=<1 1>, shear=<0 0>, translation=<-10 0>}
   }
   slot {class=OverlaySlot, properties {horizontalAlignment=HAlign_Right, verticalAlignment=VAlign_Center}}
  }
  Button_PlayerAugmentTimer {
   widgetClassPath=/Game/Athena/UI/Matchmaking/SimpleCommonButton
   bInherited=true
   bIsVariable=true
   parent=Overlay_34
   widget=Button_PlayerAugmentTimer
   properties {
    onButtonBaseClicked=(&self:WidgetTree.Button_PlayerAugmentTimer.HandleOnButtonClicked)
    onButtonBaseDoubleClicked=(&self:WidgetTree.Button_PlayerAugmentTimer.HandleOnButtonDoubleClicked)
    onButtonBaseHovered=(&self:WidgetTree.Button_PlayerAugmentTimer.HandleOnButtonHovered)
    onButtonBaseUnhovered=(&self:WidgetTree.Button_PlayerAugmentTimer.HandleOnButtonUnhovered)
    onSelectedChangedBase=(&self:WidgetTree.Button_PlayerAugmentTimer.HandleOnSelectedChanged)
    colorAndOpacity=<rgba 1 1 1 0>
   }
   slot {class=OverlaySlot, properties {horizontalAlignment=HAlign_Fill, verticalAlignment=VAlign_Fill}}
  }
  Image_Background {widgetClassPath=&Image, bInherited=true, widget=null}
  TutorialHighlight {widgetClassPath=&TutorialHighlight}
 ]
}
